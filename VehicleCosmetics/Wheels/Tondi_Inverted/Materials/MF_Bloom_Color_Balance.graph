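# Material function for the Tondi_Inverted wheel cosmetic that rebalances a custom paint color's intensity, named for bloom color balance.
# It classifies the color by RGB sum: bright paints get a brightness multiplier, others are normalized to a fixed vector length.
# A static switch, on by default, adds a darkness multiplier for near-black paints.
# The particle variant instead lerps from darkened to normalized color by the clamped sum, avoiding the threshold branches.
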MaterialFunction /VehicleCosmetics/Wheels/Tondi_Inverted/Materials/MF_Bloom_Color_Balance
graph {nodes [
  n1 = FunctionInput(Preview=Constant3Vector() {constant=<rgba 1 1 1 0>}) {inputName=CustomColor}
  n2 = Add(A=Add(A=ComponentMask(None=@n1) {r=true}, B=ComponentMask(None=@n1) {g=true}), B=ComponentMask(None=@n1) {b=true})
  n3 = Constant() {r=2}
  n4 = Multiply(
   A=FunctionInput(Preview=Constant() {r=1}) {bUsePreviewValueAsDefault=true, inputName=WhitePaint_Brightness, inputType=FunctionInput_Scalar}
   B=@n1
  )
  n5 = Multiply(
   A=FunctionInput(Preview=Constant() {r=0.5}) {bUsePreviewValueAsDefault=true, inputName=BlackPaint_Darkness, inputType=FunctionInput_Scalar}
   B=@n1
  )
  n6 = Multiply(A=Constant3Vector() {constant=<rgba 0.5774 0.5774 0.5774 0>}, B=Normalize(VectorInput=@n1))
  n7 = If(A=@n2, B=@n3, "A > B"=@n4, "A == B"=@n4, "A < B"=If(A=@n2, B=Constant() {r=0.2}, "A > B"=@n6, "A == B"=@n5, "A < B"=@n5))
  n8 = If(A=@n2, B=@n3, "A > B"=@n4, "A == B"=@n4, "A < B"=@n6)
  n9 = FunctionInput(Preview=StaticBool() {value=true}) {bUsePreviewValueAsDefault=true, inputName=HasBlackPaintVersion, inputType=FunctionInput_StaticBool}
  n10 = StaticBool()
  n11 = Clamp(None=@n2)
  FunctionOutput(None=StaticSwitch(
    True=LinearInterpolate(A=@n5, B=@n6, Alpha=@n11)
    False=StaticSwitch(True=@n7, False=@n8, Value=@n9)
    Value=FunctionInput(Preview=@n10) {bUsePreviewValueAsDefault=true, inputName=UsedForParticles, inputType=FunctionInput_StaticBool}
   ))
 ]}
