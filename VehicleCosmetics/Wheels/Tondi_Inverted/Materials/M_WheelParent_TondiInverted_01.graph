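# Parent material for the Tondi Inverted vehicle wheel: a masked, skeletal-mesh surface with black base color and emissive-driven looks.
# Vertex colors R and G blend rim lights, animated noise-and-trail tire glow and panning-cloud knife glow, mostly tinted by CustomColor, plus colored mask lines.
# Offset-function calls fake infinity-mirror depth in rim and tire layers.
# An alpha texture feeds LOD-mask opacity, and the attributes pass through shared vehicle damage and de-rez transition functions.
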Material /VehicleCosmetics/Wheels/Tondi_Inverted/Materials/M_WheelParent_TondiInverted_01
graph {
 properties {
  bAutomaticallySetUsageInEditor=false
  bUseMaterialAttributes=true
  bUsedWithSkeletalMesh=true
  blendMode=BLEND_Masked
  usedShadingModels=Default Lit
 }
 nodes [
  n1 = ScalarParameter() {defaultValue=1, parameterName=E_Line_Intensity}
  n2 = TextureSampleParameter2D() {
   parameterName=Mask_E
   samplerType=SAMPLERTYPE_LinearColor
   texture=/VehicleCosmetics/Wheels/Tondi_Inverted/Textures/SK_Wheel_Tondi_Inverted_Mask
  }
  n3 = Multiply(A=@n2.R, B=VectorParameter() {parameterName=E_LIne_Color, defaultValue=<rgba 0.2865 0.3608 1 1>}.RGB)
  n4 = VertexColor()
  n5 = TextureObject() {samplerType=SAMPLERTYPE_LinearColor, texture=/VehicleCosmetics/Wheels/Tondi_Inverted/Textures/SK_Wheel_Tondi_Inverted_E_SE2}
  n6 = ScalarParameter() {defaultValue=0.1, parameterName=BumpOffsetAmount}
  n7 = ScalarParameter() {defaultValue=0.01, parameterName=GradientDarkness}
  n8 = VectorParameter() {parameterName=CustomColor, defaultValue=<rgba 0.11 0.02 0.26 0.8>}
  n9 = Multiply(
   A=ComponentMask(None=MaterialFunctionCall!(Texture=@n5, Offset=@n6, UVs, RotateUVs, GradientFalloff, GradientDarkness=@n7) {materialFunction=/VehicleCosmetics/Wheels/ClockGear_SE/Materials/MF_Infinity_OffsetFunction}.Result) {g=true}
   B=Multiply(A=Constant() {r=3}, B=@n8.RGB)
  )
  n10 = ScalarParameter() {defaultValue=1, parameterName=SelectionFloorFade}
  n11 = Clamp(None=@n10)
  n12 = TextureSampleParameter2D() {parameterName=Mask, samplerType=SAMPLERTYPE_LinearColor, texture=/VehicleCosmetics/Wheels/Tondi_Inverted/Textures/SK_Wheel_Tondi_Inverted_E2}
  n13 = Multiply(A=ScalarParameter() {defaultValue=3, parameterName=GlowBrightness}, B=@n12.R)
  n14 = Constant()
  n15 = Add(
   A=StaticSwitchParameter(True=Multiply(A=@n9, B=@n11), False=Constant()) {defaultValue=true, parameterName=InfinityMirrorRim}
   B=Multiply(
    A=StaticSwitchParameter(True=@n13, False=@n14) {defaultValue=true, parameterName=RimLights}
    B=MaterialFunctionCall!(BlackPaint_Darkness, WhitePaint_Brightness, HasBlackPaintVersion, UsedForParticles, CustomColor=@n8.RGB) {materialFunction=&/MF_Bloom_Color_Balance}.Result
   )
  )
  n16 = Multiply(
   A=Desaturation(None=StaticComponentMaskParameter(None=@n12.R) {defaultG=true, defaultR=true, parameterName=Tire_Rgb_Glow}) {luminanceFactors=<rgba 0.3 0.59 0.11 0>}
   B=@n12.R
  )
  n17 = ScalarParameter() {defaultValue=3, parameterName=TireGlowBrightness}
  n18 = Multiply(
   A=Desaturation(None=StaticComponentMaskParameter(None=@n12.R) {defaultR=true, parameterName=Tire_Rgb_Color}) {luminanceFactors=<rgba 0.3 0.59 0.11 0>}
   B=@n12.R
  )
  n19 = Constant() {r=1}
  n20 = Add(A=Multiply(A=Multiply(A=@n16, B=@n17), B=@n8.RGB), B=Multiply(A=Multiply(A=@n18, B=@n19), B=@n8.RGB))
  n21 = TextureCoordinate() {uTiling=1.3}
  n22 = Rotator(Coordinate=Panner(Coordinate=@n21) {speedY=1}, Time=Constant() {r=1}) {speed=1.57}
  n23 = TextureSample(UVs=Panner(Coordinate=@n21) {speedY=0.68}) {texture=/VehicleCosmetics/Wheels/_SharedTextures/FX/Universal_Noise_HeightMap_08}
  n24 = Constant() {r=0.4}
  n25 = TextureSample(UVs=Panner(Coordinate=TextureCoordinate() {uTiling=1.1, vTiling=1.689}) {speedY=0.4}) {texture=/VehicleCosmetics/Wheels/_SharedTextures/FX/DistortNoise_02_RGB_T}
  n26 = Constant() {r=1}
  n27 = Add(A=Multiply(A=@n23.R, B=@n24), B=Multiply(A=Multiply(A=@n25.G, B=@n24), B=@n26))
  n28 = Add(
   A=Multiply(
    A=TextureSample(UVs=@n22) {samplerType=SAMPLERTYPE_LinearColor, texture=/VehicleCosmetics/Wheels/Tondi_Inverted/Textures/TechnoTrails_RGB_T}.R
    B=ScalarParameter() {defaultValue=1, parameterName=MovingLine_Intensity}
   )
   B=@n27
  )
  n29 = Constant() {r=0.4}
  n30 = TextureSample(UVs=Panner(Coordinate=TextureCoordinate() {uTiling=1.3, vTiling=2}) {speedY=1}) {texture=&DistortNoise_02_RGB_T}
  n31 = Multiply(A=@n30.G, B=@n29)
  n32 = Add(A=@n27, B=Add(A=Multiply(A=@n23.RGB, B=@n29), B=Multiply(A=@n31, B=@n26)))
  n33 = Constant() {r=5}
  n34 = Panner(Coordinate=TextureCoordinate() {uTiling=1.3, vTiling=2}) {speedY=1.5}
  n35 = Multiply(A=Multiply(A=@n33, B=TextureSample(UVs=@n34) {texture=/VehicleCosmetics/Wheels/_SharedTextures/FX/Noise_Spots_RGB_T}.R), B=@n30.B)
  n36 = Multiply(A=@n35, B=@n33)
  n37 = Multiply(A=@n20, B=Add(A=Add(A=@n28, B=@n32), B=Add(A=@n36, B=@n35)))
  n38 = Constant() {r=0.5}
  n39 = ComponentMask(None=MaterialFunctionCall(
    Texture=TextureObject() {samplerType=SAMPLERTYPE_LinearColor, texture=/VehicleCosmetics/Wheels/Tondi_Inverted/Textures/SK_Wheel_Tondi_Inverted_E_SE}
    Offset=@n6
    GradientDarkness=@n7
   ) {materialFunction=&MF_Infinity_OffsetFunction}.Result) {g=true}
  n40 = Multiply(A=Constant() {r=3}, B=@n8.RGB)
  n41 = Add(A=Multiply(A=LinearInterpolate(A=@n20, B=@n37, Alpha=@n38), B=@n19), B=Multiply(A=Multiply(A=@n39, B=@n40), B=Clamp(None=@n10)))
  n42 = Add(A=@n4.R, B=@n4.R)
  n43 = Multiply(A=TextureCoordinate(), B=ScalarParameter() {defaultValue=1, parameterName=KnifeTile})
  n44 = Multiply(
   A=Multiply(A=NamedRerouteUsage(), B=Constant() {r=1})
   B=TextureSample(UVs=Panner(Coordinate=@n43) {speedX=0.158}) {samplerType=SAMPLERTYPE_LinearColor, texture=/VehicleCosmetics/Wheels/Tondi/New/Textures/Clouds_RGB_02_T}.B
  )
  n45 = Multiply(
   A=TextureSample(UVs=Panner(Coordinate=@n43) {speedX=0.2}) {samplerType=SAMPLERTYPE_LinearColor, texture=&Clouds_RGB_02_T}.R
   B=Constant() {r=1}
  )
  n46 = Multiply(A=Multiply(A=Multiply(A=@n44, B=@n45), B=@n8.RGB), B=ScalarParameter() {defaultValue=1, parameterName=GlowEffect_Knive})
  n47 = Add(
   A=LinearInterpolate(B=Multiply(A=@n1, B=@n3), Alpha=OneMinus(None=@n4.R))
   B=LinearInterpolate(A=LinearInterpolate(A=@n15, B=@n41, Alpha=@n42), B=LinearInterpolate(A=@n15, B=@n46, Alpha=@n4.G), Alpha=@n4.G)
  )
  n48 = TextureSampleParameter2D() {
   parameterName=AlphaMask
   samplerType=SAMPLERTYPE_LinearColor
   texture=/VehicleCosmetics/Wheels/Tondi_Inverted/Textures/SK_Wheel_Tondi_Inverted_Alpha_LOD
  }
  n49 = Add(A=@n4.R, B=@n4.R)
  n50 = LinearInterpolate(
   A=Constant() {r=1}
   B=ScalarParameter() {defaultValue=1, parameterName=LOds_Mask}
   Alpha=LinearInterpolate(A=Constant(), B=@n48.A, Alpha=OneMinus(None=@n49))
  )
  n51 = Constant()
  MaterialFunctionCall(In=MaterialFunctionCall(Material=SetMaterialAttributes!(MaterialAttributes, Emissive Color=@n47, Opacity=@n50, Base Color=@n51) {attributeSetTypes [#1, #2, #3]}) {materialFunction=&MF_BaseVehicle_Damage}.Result) {materialFunction=&MF_VehicleDeRez_Transition}
  Multiply(A=@n4.G)
  NamedRerouteDeclaration(None=@n12.B) {name=Rim_M_Blue, nodeColor=<rgba 0 0.6484 1 1>}
  NamedRerouteDeclaration(None=@n12.G) {name=Rim_M_Green, nodeColor=<rgba 0.7109 0 1 1>}
  NamedRerouteUsage()
 ]
}
