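# Material function for an animated, tintable emissive blade effect (Blade Effect 02) on the Tondi_Inverted wheel cosmetic.
# Swirl, noise and spot textures spin at different Rotator speeds and are layered through adds and multiplies.
# A static light-burst channel, scaled by a blade intensity parameter, shapes the glow before custom color (default blue) and overall intensity apply.
# ColorBE02 and OpacityBE02 both output that same tinted result.
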MaterialFunction /VehicleCosmetics/Wheels/Tondi_Inverted/Materials/MF_BaleEffect02
graph {nodes [
  n1 = Constant() {r=1}
  n2 = TextureCoordinate()
  n3 = TextureSample(UVs=Rotator(Coordinate=@n2) {speed=-3}) {samplerType=SAMPLERTYPE_LinearColor, texture=/VehicleCosmetics/Wheels/Tondi_Inverted/Textures/Swirl_RGB_03_T}
  n4 = Rotator(Coordinate=@n2) {speed=-0.75}
  n5 = TextureSample(UVs=@n4) {samplerType=SAMPLERTYPE_LinearColor, texture=&Swirl_RGB_03_T}
  n6 = Multiply(
   A=ScalarParameter() {defaultValue=1.4, group=Blade_Effect_02, parameterName=Blade_Intensity_BE02}
   B=TextureSample() {samplerType=SAMPLERTYPE_LinearColor, texture=/VehicleCosmetics/Wheels/_SharedTextures/FX/LightBurstAlpha_RGBA}.B
  )
  n7 = Add(A=Add(A=Multiply(A=@n1, B=@n3.B), B=Multiply(A=@n1, B=@n5.G)), B=@n6)
  n8 = Constant() {r=1}
  n9 = TextureCoordinate()
  n10 = TextureSample(UVs=Rotator(Coordinate=@n9) {speed=-4}) {samplerType=SAMPLERTYPE_LinearColor, texture=&Swirl_RGB_03_T}
  n11 = Constant() {r=0.2}
  n12 = TextureSample(UVs=Rotator(Coordinate=@n9) {speed=-2}) {samplerType=SAMPLERTYPE_LinearColor, texture=&Swirl_RGB_03_T}
  n13 = Constant() {r=0.4}
  n14 = TextureSample(UVs=Rotator(Coordinate=@n2)) {texture=/VehicleCosmetics/Wheels/_SharedTextures/FX/DistortNoise_02_RGB_T}
  n15 = Constant() {r=2}
  n16 = Add(A=Multiply(A=@n12.B, B=@n13), B=Multiply(A=Multiply(A=@n14.G, B=@n13), B=@n15))
  n17 = Constant() {r=0.4}
  n18 = TextureSample(UVs=@n4) {texture=&DistortNoise_02_RGB_T}
  n19 = Add(A=Multiply(A=@n12.G, B=@n17), B=Multiply(A=Multiply(A=@n18.G, B=@n17), B=@n15))
  n20 = TextureSample(UVs=Rotator(Coordinate=@n2) {speed=-0.9}) {texture=/VehicleCosmetics/Wheels/_SharedTextures/FX/Noise_Spots_RGB_T}
  n21 = Constant() {r=10}
  n22 = TextureSample(UVs=Rotator(Coordinate=@n2) {speed=-1}) {texture=&Noise_Spots_RGB_T}
  n23 = Add(A=Add(A=Add(A=@n10.G, B=@n16), B=Add(A=@n16, B=@n19)), B=Multiply(A=Multiply(A=@n15, B=@n20.R), B=Multiply(A=@n21, B=@n22.R)))
  n24 = OneMinus(None=Multiply(A=@n8, B=@n10.R))
  n25 = Add(A=Multiply(A=Multiply(A=@n10.G, B=@n8), B=Multiply(A=@n11, B=@n23)), B=Multiply(A=@n23, B=Multiply(A=@n24, B=@n6)))
  n26 = Multiply(
   A=Multiply(
    A=Multiply(A=@n25, B=@n7)
    B=VectorParameter() {group=Blade_Effect_02, parameterName=CustomColor_BE_02, defaultValue=<rgba 0.07 0.14 1 1>}.RGB
   )
   B=ScalarParameter() {defaultValue=3, group=Blade_Effect_02, parameterName=Overall_Emissve_Intensity_BE02}
  )
  FunctionOutput(None=@n26) {outputName=ColorBE02}
  FunctionOutput(None=@n26) {outputName=OpacityBE02}
 ]}
