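# Translucent parent material for the Tondi Inverted vehicle wheel's animated blade effect, with glow tinted by PaintColor.
# Layers of panning noise and trail textures are masked by line-mask channels into blade and middle-ring terms, each with an intensity parameter.
# Vertex color green blends MF_BaleEffect02's color and opacity toward this glow, which drives both emissive and opacity.
# The attributes then pass through shared vehicle damage and de-rez transition functions.
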Material /VehicleCosmetics/Wheels/Tondi_Inverted/Materials/M_WheelParent_TondiInverted_BladeEffect
graph {
 properties {
  bAutomaticallySetUsageInEditor=false
  bUseMaterialAttributes=true
  bUsedWithSkeletalMesh=true
  blendMode=BLEND_Translucent
  previewMesh=/VehicleCosmetics/Wheels/Dimen/Meshes/SK_Wheel_Dimen
  usedShadingModels=Default Lit
 }
 nodes [
  Constant()
  n1 = MaterialFunctionCall() {materialFunction=&/MF_BaleEffect02}
  n2 = TextureSample() {samplerType=SAMPLERTYPE_LinearColor, texture=/VehicleCosmetics/Wheels/Tondi_Inverted/Textures/SK_Wheel_Tondi_Inverted_LineMasks}
  n3 = Constant() {r=1}
  n4 = Constant() {r=0.2}
  n5 = TextureCoordinate() {uTiling=1.3}
  n6 = TextureSample(UVs=Panner(Coordinate=@n5) {speedX=0.68}) {texture=/VehicleCosmetics/Wheels/Tondi_Inverted/Textures/Universal_Noise_42_RGB}
  n7 = Constant() {r=0.4}
  n8 = TextureSample(UVs=Panner(Coordinate=TextureCoordinate() {uTiling=1.1, vTiling=1.689}) {speedY=0.4}) {texture=/VehicleCosmetics/Wheels/_SharedTextures/FX/DistortNoise_02_RGB_T}
  n9 = Constant() {r=2}
  n10 = Add(A=Multiply(A=@n6.R, B=@n7), B=Multiply(A=Multiply(A=@n8.G, B=@n7), B=@n9))
  n11 = TextureSample(UVs=Panner(Coordinate=@n5) {speedX=2}) {samplerType=SAMPLERTYPE_LinearColor, texture=/VehicleCosmetics/Wheels/Tondi_Inverted/Textures/TechnoTrails_RGB_T}
  n12 = Constant() {r=0.4}
  n13 = TextureSample(UVs=Panner(Coordinate=TextureCoordinate() {uTiling=1.3, vTiling=2}) {speedX=1}) {texture=&DistortNoise_02_RGB_T}
  n14 = Add(A=Multiply(A=@n6.RGB, B=@n12), B=Multiply(A=Multiply(A=@n13.G, B=@n12), B=@n9))
  n15 = Constant() {r=0.4}
  n16 = TextureCoordinate() {uTiling=0.5, vTiling=0.5}
  n17 = Multiply(A=@n15, B=TextureSample(UVs=Panner(Coordinate=@n16) {speedX=1}) {texture=/VehicleCosmetics/Wheels/_SharedTextures/FX/Noise_Spots_RGB_T}.R)
  n18 = Constant() {r=5}
  n19 = Add(A=Add(A=Add(A=@n11.R, B=@n10), B=Add(A=@n10, B=@n14)), B=Add(A=@n15, B=Multiply(A=@n17, B=@n18)))
  n20 = OneMinus(None=Multiply(A=@n3, B=@n2.A))
  n21 = Multiply(A=@n2.G, B=ScalarParameter() {defaultValue=1, parameterName=Blade_Intensity})
  n22 = Add(A=Multiply(A=Multiply(A=@n2.B, B=@n3), B=Multiply(A=@n4, B=@n19)), B=Multiply(A=@n19, B=Multiply(A=@n20, B=@n21)))
  n23 = TextureCoordinate() {uTiling=1.3}
  n24 = Panner(Coordinate=@n23) {speedY=0.4}
  n25 = TextureSample(UVs=Panner(Coordinate=@n23) {speedY=0.68}) {texture=/VehicleCosmetics/Wheels/_SharedTextures/FX/Universal_Noise_HeightMap_08}
  n26 = Constant() {r=0.4}
  n27 = TextureSample(UVs=Panner(Coordinate=TextureCoordinate()) {speedY=0.4}) {texture=&DistortNoise_02_RGB_T}
  n28 = Add(A=Multiply(A=@n25.RGB, B=@n26), B=Multiply(A=Multiply(A=@n27.B, B=@n26), B=Constant() {r=2}))
  n29 = Constant() {r=0.4}
  n30 = Multiply(A=@n25.RGB, B=@n29)
  n31 = Panner(Coordinate=TextureCoordinate()) {speedX=0.1, speedY=0.8}
  n32 = Multiply(
   A=Multiply(A=TextureSample(UVs=@n31) {texture=&DistortNoise_02_RGB_T}.G, B=@n29)
   B=Constant() {r=2}
  )
  n33 = Add(
   A=Add(A=TextureSample(UVs=@n24) {texture=&DistortNoise_02_RGB_T}.G, B=@n28)
   B=Add(A=@n28, B=Add(A=@n30, B=@n32))
  )
  n34 = ComponentMask(None=@n23) {g=true}
  n35 = Multiply(A=Multiply(A=Multiply(A=@n33, B=@n34), B=@n2.R), B=ScalarParameter() {defaultValue=1, parameterName=Middle_Ring_Intensity})
  n36 = Multiply(
   A=Multiply(A=Add(A=@n22, B=@n35), B=VectorParameter() {parameterName=PaintColor, defaultValue=<rgba 0.3 0.06 0.73 1>}.RGB)
   B=ScalarParameter() {defaultValue=5, parameterName=Overall_Emissve_Intensity}
  )
  n37 = VertexColor()
  n38 = LinearInterpolate(A=@n1.ColorBE02, B=@n36, Alpha=@n37.G)
  n39 = LinearInterpolate(A=@n1.OpacityBE02, B=@n36, Alpha=@n37.G)
  MaterialFunctionCall(In=MaterialFunctionCall(Material=SetMaterialAttributes!(MaterialAttributes, Emissive Color=@n38, Opacity=@n39) {attributeSetTypes [#1, #2]}) {materialFunction=&MF_BaseVehicle_Damage}.Result) {materialFunction=&MF_VehicleDeRez_Transition}
  Multiply(A=@n37.G)
 ]
}
